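# Niagara bubble effect for the Seesaw Slipper backpack cosmetic, built as one GPU sprite emitter of small, semi-transparent bluish particles.
# Bubbles spawn in a tilted box, then drift under local acceleration, drag and point attraction around three sphere blocking volumes.
# Three inverted region masks multiply into a sprite size scale, and a box volume kills strays, keeping bubbles confined.
# A curve-blended random range drives a dynamic material parameter over time.
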NiagaraSystem /BRCosmetics/Accessories/FORT_Backpacks/Backpack_M_MED_Seesaw_Slipper/FX/NS_Seesaw_Slipper_Bubbles
dependencies {
 usedDataInterfaces [&NiagaraDataInterfaceDebugDraw, &NiagaraDataInterfaceCurve]
 usedDynamicInputs []
 usedModules [&SystemState]
}
emitters {Bubbles {
  simTarget=GPUComputeSim
  emitterUpdateScript [
   &EmitterState()
   &SpawnRate(SpawnRate=&RandomRangeFloat(Minimum=75, Maximum=150), Use Spawn Probability=false)
  ]
  particleSpawnScript [
   &InitializeParticle(
    Color=&MakeLinearColorFromVectorAndFloat(
     "Vector (RGB)"=&MakeVectorFromLinearColorRGB(Color=<rgba 0.0456 0.0876 0.1146 1>)
     "Float (Alpha)"=&RandomRangeFloat(Minimum=0.333, Maximum=0.75)
    )
    Lifetime=4
    Sprite Rotation Mode=::NewEnumerator2
    Sprite Size Mode=::NewEnumerator1
    Uniform Sprite Size Max=1.8
    Uniform Sprite Size Min=0.5
    UsePositionOffset=false
    Write Material Random=true
   )
   &BoxLocation(
    Box Offset=<0 0 5>
    Box Size=<10 20 0>
    MaskSpawnGroup=false
    Module.Override Seed=false
    Normalized Box Rotation Angle=-0.2
    Rotate Box=true
    Rotate Box On Axis:Vector3f=<0 1 0>
    Surface Only=false
   )
  ]
  particleUpdateScript [
   &ParticleState()
   &AccelerationForce(Acceleration=&RandomRangeVector(Minimum=<5 -5 -1>, Maximum=<8 5 -1>), Coordinate Space=::Local)
   &Drag(Drag=&RandomRangeFloat(Minimum=0.5, Maximum=2), Use Linear Drag=true, Use Rotational Drag=false)
   &PointAttractionForce(
    Attraction Radius=20
    Attraction Strength=5
    Attractor Position Offset:Vector3f=<20 0 8>
    Attractor Position Offset Coordinate Space:ENiagaraCoordinateSpace=::Local
    Falloff Exponent=0.125
    Kill Radius=5
    Kill Within Radius=false
    Module.Use Falloff=true
    UseAttractorOffset=true
   )
   &SolveForcesAndVelocity(Clamp Velocity=false, Limit Acceleration=false)
   &self:SphereBlockingVolume(Pivot Offset:Vector3f=<5 -3 0>, Sphere Origin:Vector3f=/Niagara/DynamicInputs/Helpers/SimulationPosition(), Sphere Radius=4) {module=SphereBlockingVolume}
   &self:SphereBlockingVolume(Pivot Offset:Vector3f=<10 6 5>, Sphere Origin:Vector3f=&SimulationPosition(), Sphere Radius=3) {module=SphereBlockingVolume001}
   &self:SphereBlockingVolume(Pivot Offset:Vector3f=<15 0 6>, Sphere Origin:Vector3f=&SimulationPosition(), Sphere Radius=5.25) {module=SphereBlockingVolume002}
   &CalculateAccurateVelocity()
   &RegionMask(Invert Region=true, Region Falloff=0, Region Type:ENiagaraDebugDrawShape=::NewEnumerator2, UseOffset=true, UseRotation=false)
   &RegionMask(
    Invert Region=true
    Offset:Vector3f=<0 5 0>
    Region Falloff=0
    Region Type:ENiagaraDebugDrawShape=::NewEnumerator2
    Rotation:Vector3f=<0 0 0.175>
    Transform Order:ENiagaraRegionTransformOrder=::NewEnumerator5
    UseOffset=true
    UseRotation=true
   ) {module=RegionMask001}
   &RegionMask(
    Invert Region=true
    Offset:Vector3f=<0 -5 0>
    Region Falloff=0
    Region Type:ENiagaraDebugDrawShape=::NewEnumerator2
    Rotation:Vector3f=<0 0 -0.175>
    Transform Order:ENiagaraRegionTransformOrder=::NewEnumerator5
    UseOffset=true
    UseRotation=true
   ) {module=RegionMask002}
   /SetVariables(Particles.RegionControl:NiagaraFloat=&Multiply_Float(
     A=&Multiply_Float(A=$Particles.RegionMask.Float, B=$Particles.RegionMask001.Float)
     B=$Particles.RegionMask002.Float
    )) {isSetParameters=true}
   &DynamicMaterialParameters(
    Index 0 Param 1=&Lerp_Float(
     A=&RandomRangeFloat(Minimum=0, Maximum=0.2)
     B=&RandomRangeFloat(Minimum=0.8, Maximum=1)
     Alpha=&FloatFromCurve(FloatCurve=DataInterface() {propertyValues {Curve {keys [<at 0 1 RCIM_Cubic>,<at 0.5 0 RCIM_Cubic>,<at 1 1 RCIM_Cubic>]}}})
    )
    Module.Param0WriteEnabled=true
    Module.Param1WriteEnabled=true
    Module.Param2WriteEnabled=true
    Module.Param3WriteEnabled=true
   )
   &ScaleSpriteSize(Scale Sprite Size Mode=::NewEnumerator0, Uniform Scale Factor=$Particles.RegionControl) {module=ScaleSpriteSize001}
   &KillParticlesInVolume(Box Size:Vector3f=<50 12 20>, Invert Volume=true, Kill Shape=::NewEnumerator1)
  ]
  rendererClasses [&NiagaraSpriteRendererProperties]
  renderers [{rendererClass=&NiagaraSpriteRendererProperties}]
 }}
summary {
 systemName=NS_Seesaw_Slipper_Bubbles
 emitters [{emitterName=Bubbles, simTarget=GPUComputeSim, rendererClasses [&NiagaraSpriteRendererProperties]}]
 userVariables [User.SepTrans {type=NiagaraFloat, defaultValue=1, description=""}]
}
