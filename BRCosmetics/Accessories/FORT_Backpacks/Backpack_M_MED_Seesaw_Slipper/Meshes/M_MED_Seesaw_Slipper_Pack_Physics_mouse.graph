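# Physics asset for the Seesaw Slipper back bling cosmetic (mouse file), covering the pack and a three-bone dynamic hook chain.
# Each hook bone has a capsule body, with a much larger offset one on the last link, while the pack carries two large capsules.
# Constraints link the hooks in series back to the pack, all sharing the same swing and twist limit angles.
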PhysicsAsset /BRCosmetics/Accessories/FORT_Backpacks/Backpack_M_MED_Seesaw_Slipper/Meshes/M_MED_Seesaw_Slipper_Pack_Physics_mouse
bodies {
 dyn_hook_1 [{shapeType=Capsule, length=0.505, radius=0.505, shapeName=dyn_hook_1_capsule, rotation=<rot 90 0 0>}]
 dyn_hook_2 [{shapeType=Capsule, length=0.4911, radius=0.6791, shapeName=dyn_hook_2_capsule, rotation=<rot 90 0 0>}]
 dyn_hook_3 [{shapeType=Capsule, length=1.3355, radius=4.0532, shapeName=dyn_hook_3_capsule, center=<4 -3 0>, rotation=<rot 90 0 0>}]
 pack [
  {shapeType=Capsule, length=12.6236, radius=7.4474, shapeName=pack_capsule, center=<-3.8349 -0.0081 -6.8104>, rotation=<rot 0 -9.2737 0>}
  {shapeType=Capsule, length=15, radius=15, shapeName=pack_capsule, center=<-14.3385 17.264 0.0004>, rotation=<rot 0 6.6229 0>}
 ]
}
constraints [
 {bone1Name=dyn_hook_1, bone2Name=pack, swing1LimitDegrees=90, swing2LimitDegrees=90, twistLimitDegrees=30}
 {bone1Name=dyn_hook_2, bone2Name=dyn_hook_1, swing1LimitDegrees=90, swing2LimitDegrees=90, twistLimitDegrees=30}
 {bone1Name=dyn_hook_3, bone2Name=dyn_hook_2, swing1LimitDegrees=90, swing2LimitDegrees=90, twistLimitDegrees=30}
]
