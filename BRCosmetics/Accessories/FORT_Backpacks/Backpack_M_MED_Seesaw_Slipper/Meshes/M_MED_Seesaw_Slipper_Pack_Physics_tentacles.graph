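# Physics asset for the Seesaw Slipper backpack cosmetic that defines simulation bodies for its dangling tentacles.
# The pack body is built from three large capsules, while four tentacle chains (A to D) of two to four bones use small capsules.
# Each chain is constrained root to tip from the pack, with tight swing and twist limits.
# The final joint of each chain gets a much wider first swing limit, so the tips bend more freely.
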PhysicsAsset /BRCosmetics/Accessories/FORT_Backpacks/Backpack_M_MED_Seesaw_Slipper/Meshes/M_MED_Seesaw_Slipper_Pack_Physics_tentacles
bodies {
 dyn_tentacle_A1 [{shapeType=Capsule, length=1, radius=1, shapeName=dyn_tentacle_A1_capsule}]
 dyn_tentacle_A2 [{shapeType=Capsule, length=1, radius=1, shapeName=dyn_tentacle_A2_capsule}]
 dyn_tentacle_A3 [{shapeType=Capsule, length=1, radius=1, shapeName=dyn_tentacle_A3_capsule}]
 dyn_tentacle_B1 [{shapeType=Capsule, length=1, radius=1, shapeName=dyn_tentacle_B1_capsule}]
 dyn_tentacle_B2 [{shapeType=Capsule, length=1, radius=1, shapeName=dyn_tentacle_B2_capsule}]
 dyn_tentacle_C1 [{shapeType=Capsule, length=1, radius=1, shapeName=dyn_tentacle_C1_capsule}]
 dyn_tentacle_C2 [{shapeType=Capsule, length=1, radius=1, shapeName=dyn_tentacle_C2_capsule}]
 dyn_tentacle_C3 [{shapeType=Capsule, length=1, radius=1, shapeName=dyn_tentacle_C3_capsule}]
 dyn_tentacle_C4 [{shapeType=Capsule, length=0.25, radius=1, shapeName=dyn_tentacle_C4_capsule}]
 dyn_tentacle_D1 [{shapeType=Capsule, length=1, radius=1, shapeName=dyn_tentacle_D1_capsule}]
 dyn_tentacle_D2 [{
   shapeType=Capsule
   length=3.25
   radius=0.75
   shapeName=dyn_tentacle_D2_capsule
   center=<2.6194 0 0.3397>
   rotation=<rot -82.6106 0.0001 -0.0002>
  }]
 dyn_tentacle_D3 [{shapeType=Capsule, length=1, radius=1, shapeName=dyn_tentacle_D3_capsule}]
 dyn_tentacle_D4 [{shapeType=Capsule, length=1, radius=1, shapeName=dyn_tentacle_D4_capsule}]
 pack [
  {shapeType=Capsule, length=23.75, radius=34.25, shapeName=pack_capsule, center=<0.5786 -46.7573 -8.0591>, rotation=<rot 0 -11.8352 0>}
  {
   shapeType=Capsule
   length=15
   radius=25.75
   shapeName=pack_capsule
   center=<-0.0001 -0.041 -41.7711>
   rotation=<rot 86.7926 -90.0237 -89.9997>
  }
  {shapeType=Capsule, length=23, radius=31, shapeName=pack_capsule, center=<-37.3372 -19.8337 -8.0691>}
 ]
}
constraints [
 {bone1Name=dyn_tentacle_B2, bone2Name=dyn_tentacle_B1, swing1LimitDegrees=50, swing2LimitDegrees=2, twistLimitDegrees=5}
 {bone1Name=dyn_tentacle_A1, bone2Name=pack, swing1LimitDegrees=10, swing2LimitDegrees=2, twistLimitDegrees=5}
 {bone1Name=dyn_tentacle_A2, bone2Name=dyn_tentacle_A1, swing1LimitDegrees=10, swing2LimitDegrees=10, twistLimitDegrees=5}
 {bone1Name=dyn_tentacle_A3, bone2Name=dyn_tentacle_A2, swing1LimitDegrees=50, swing2LimitDegrees=2, twistLimitDegrees=5}
 {bone1Name=dyn_tentacle_C1, bone2Name=pack, swing1LimitDegrees=10, swing2LimitDegrees=2, twistLimitDegrees=5}
 {bone1Name=dyn_tentacle_C2, bone2Name=dyn_tentacle_C1, swing1LimitDegrees=10, swing2LimitDegrees=2, twistLimitDegrees=5}
 {bone1Name=dyn_tentacle_C3, bone2Name=dyn_tentacle_C2, swing1LimitDegrees=10, swing2LimitDegrees=2, twistLimitDegrees=5}
 {bone1Name=dyn_tentacle_C4, bone2Name=dyn_tentacle_C3, swing1LimitDegrees=50, swing2LimitDegrees=2, twistLimitDegrees=5}
 {bone1Name=dyn_tentacle_D1, bone2Name=pack, swing1LimitDegrees=10, swing2LimitDegrees=2, twistLimitDegrees=5}
 {bone1Name=dyn_tentacle_D2, bone2Name=dyn_tentacle_D1, swing1LimitDegrees=10, swing2LimitDegrees=2, twistLimitDegrees=5}
 {bone1Name=dyn_tentacle_D3, bone2Name=dyn_tentacle_D2, swing1LimitDegrees=10, swing2LimitDegrees=2, twistLimitDegrees=5}
 {bone1Name=dyn_tentacle_D4, bone2Name=dyn_tentacle_D3, swing1LimitDegrees=50, swing2LimitDegrees=2, twistLimitDegrees=5}
 {bone1Name=dyn_tentacle_B1, bone2Name=pack, swing1LimitDegrees=10, swing2LimitDegrees=2, twistLimitDegrees=5}
]
