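# Animation Blueprint for the Seesaw Slipper backpack cosmetic, driving idle motion and secondary physics.
# A shuffled idle and idle-break player feeds chained RigidBody simulations for tentacles and mouse, using separate physics assets.
# An additive mouse-root rotation applies only in the front end, then a Control Rig borrowed from M_MED_Freak takes `swing delta`.
# The update event eases `swing delta` toward zero; nothing here sets it otherwise.
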AnimBlueprint /BRCosmetics/Accessories/FORT_Backpacks/Backpack_M_MED_Seesaw_Slipper/Meshes/M_MED_Seesaw_Slipper_Pack_AnimBP
target_skeleton=&/M_MED_Seesaw_Slipper_Pack_Skeleton
graphs [
 AnimGraph {nodes [
   n1 = Random Sequence Player!() -> (Pose:Pose Link Structure) {
    is=title AnimGraphNode_RandomPlayer
    bShuffleMode=true
    entries [
     {
      chanceToPlay=0.6
      maxLoopCount=2
      maxPlayRate=1
      minLoopCount=0
      minPlayRate=1
      sequence=/BRCosmetics/Accessories/FORT_Backpacks/Backpack_M_MED_Seesaw_Slipper/Animation/Seesaw_Slipper_Idle_BACKPACK
      blendIn {blendOption=Linear, blendTime=0.2, customCurve=null}
     }
     {
      chanceToPlay=0.4
      maxLoopCount=2
      maxPlayRate=1
      minLoopCount=0
      minPlayRate=1
      sequence=/BRCosmetics/Accessories/FORT_Backpacks/Backpack_M_MED_Seesaw_Slipper/Animation/Seesaw_Slipper_IdleBreak_BACKPACK
      blendIn {blendOption=Linear, blendTime=0.2, customCurve=null}
     }
    ]
   }
   n2 = RigidBody(ComponentPose=Local To Component(LocalPose=@n1)) {
    bOverrideWorldGravity=true
    overridePhysicsAsset=&/M_MED_Seesaw_Slipper_Pack_Physics_tentacles
    simulationSpace=BaseBoneSpace
    alphaBoolBlend {blendOption=Linear}
    componentLinearAccScale=<0.1 0.1 0.1>
    componentLinearVelScale=<0.1 0.1 0.1>
    overrideWorldGravity {z=-980}
   }
   n3 = RigidBody(ComponentPose=@n2) {
    bOverrideWorldGravity=true
    overridePhysicsAsset=&/M_MED_Seesaw_Slipper_Pack_Physics_mouse
    simulationSpace=BaseBoneSpace
    alphaBoolBlend {blendOption=Linear}
    componentLinearAccScale=<0.31 0.31 0.31>
    componentLinearVelScale=<0.31 0.31 0.31>
    overrideWorldGravity {z=-980}
   }
   n4 = "Transform (Modify) Bone - Bone: fk_mouseRoot_1"(
    ComponentPose=@n3
    Alpha="To Float (Boolean)"(InBool=Get bIsInFrontEnd())
    Translation="0,0,0"
    Rotation="-15.000,25.281,10.000"
    Scale="1,1,1"
   ) {rotationMode=BMM_Additive, alphaBoolBlend {blendOption=Linear}, boneToModify {boneName=fk_mouseRoot_1}, rotation=<rot -15 25.281 10>}
   n5 = Control Rig!(Source=Component To Local(ComponentPose=@n4), Alpha, targeting:double=Get swing delta() -> (swing delta:double)) -> (Pose) {
    alphaBoolBlend {blendOption=Linear}
    controlRigAssetReference {blueprintRigClass=/BRCosmetics/Characters/Player/Male/Medium/Bodies/M_MED_Freak/Meshes/Parts/M_MED_Freak_FaceAcc_CtrlRig.M_MED_Freak_FaceAcc_CtrlRig_C}
   }
   Output Pose(Result=@n5) {layerGroup=None}
  ]}
 EventGraph {nodes [
   n1 = Event BlueprintUpdateAnimation()
   > Set swing delta(swing delta:double=FInterpTo(Current=Get swing delta(), Target=0, DeltaTime=@n1.DeltaTimeX, InterpSpeed=2)) -> (Output_Get:double)
   TryGetPawnOwner()
  ]}
]
variables [swing delta {type=number}]
